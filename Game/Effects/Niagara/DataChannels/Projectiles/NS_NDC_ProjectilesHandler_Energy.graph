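# Niagara handler that draws energy projectile tracers from data channel entries, using one Bullet emitter with mesh and sprite renderers.
# Each entry initializes a particle with tracer length from movement since the previous position, width by ammo type and LOD-distance scaling.
# In-flight particles live about one frame and start-position particles are killed.
# End-position particles live longer, sliding from previous position toward the endpoint while tracer length animates out.
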NiagaraSystem /Game/Effects/Niagara/DataChannels/Projectiles/NS_NDC_ProjectilesHandler_Energy
dependencies {
 usedDataInterfaces [
  &NiagaraDataInterfaceArrayFloat
  &NiagaraDataInterfaceCurve
  /Script/Niagara.NiagaraDataInterfaceDataChannelRead
 ]
 usedDynamicInputs []
 usedModules [
  &SystemState
  /Niagara/Modules/System/CompleteIfUnused
  &self:SystemSpawnScript.SystemScriptSource.SystemScriptGraph.NiagaraNodeAssignment_0.SetVariables_#1
 ]
}
emitters {Bullet {
  emitterSpawnScript [/SetVariables(Emitter.UseTeardrop=true) {isSetParameters=true}]
  emitterUpdateScript [
   &EmitterState()
   &/Spawn_NDCFortProjectiles(Max Spawn Count=1, Min Spawn Count=1, Reader:NiagaraDataInterfaceDataChannelRead=$System.NDCReader)
  ]
  particleSpawnScript [
   &/InitializeParticleFromNDC_FortProjectiles(FallbackAmmoType=2)
   /SetVariables(Particles.PositionType:NiagaraInt32=$Output.InitializeParticleFromNDC_FortProjectiles.PositionType) {isSetParameters=true}
   /SetVariables(Particles.Random=&RandomRangeFloat(
     Minimum=0
     Maximum=1
     Randomness Mode=::NewEnumerator1
     Random Seed=&ReturnParticleIDAttributes(Particle ID Attribute=::NewEnumerator1, Niagara ID=$Particles.RibbonID)
     Fixed Random Seed=true
    )) {isSetParameters=true}
   /SetVariables(Particles.LOD Distance:NiagaraFloat=$Output.InitializeParticleFromNDC_FortProjectiles.LODDistance) {isSetParameters=true}
   /SetVariables(Particles.AmmoTypeScale:NiagaraFloat=&SelectFloatFromArray(
     Direct Array Index=$Output.InitializeParticleFromNDC_FortProjectiles.AmmoType
     Float Selection Array=DataInterface() {propertyValues {FloatData [0.75,1,1.2,0.75]}}
    )) {isSetParameters=true}
   /SetVariables(Particles.IsStartPosition:NiagaraBool=&SetBoolByIntComparison(A=$Particles.PositionType, B=0, Comparison Type=::NewEnumerator4)) {isSetParameters=true}
   /SetVariables(Particles.IsEndPosition:NiagaraBool=&SetBoolByIntComparison(A=$Particles.PositionType, B=2, Comparison Type=::NewEnumerator4)) {isSetParameters=true}
   /SetVariables(Particles.OriginTime:NiagaraFloat=$Output.InitializeParticleFromNDC_FortProjectiles.OriginTime) {isSetParameters=true}
   /SetVariables(Particles.ProjectileAge:NiagaraFloat=&Subtract_Float(A=$Engine.Time, B=$Particles.OriginTime)) {isSetParameters=true}
   /SetVariables(Particles.MuzzlePosition:NiagaraPosition=$Output.InitializeParticleFromNDC_FortProjectiles.OriginPosition) {isSetParameters=true}
   /SetVariables(Particles.PreviousPosition:NiagaraPosition=$Output.InitializeParticleFromNDC_FortProjectiles.PrevPosition) {isSetParameters=true}
   /SetVariables(Particles.TargetPosition:NiagaraPosition=$Output.InitializeParticleFromNDC_FortProjectiles.Position) {isSetParameters=true}
   /SetVariables(Particles.RibbonUVDistance:NiagaraFloat=&DistanceBetweenPositions(A=$Particles.TargetPosition, B=$Particles.MuzzlePosition)) {isSetParameters=true}
   /SetVariables(Particles.Distance:NiagaraFloat=&DistanceBetweenPositions(A=$Particles.PreviousPosition, B=$Particles.TargetPosition)) {isSetParameters=true}
   /SetVariables(Particles.Velocity=&Subtract_Position(A=$Particles.TargetPosition, B=$Particles.PreviousPosition)) {isSetParameters=true}
   /SetVariables(Particles.VelocityNormalized:Vector3f=&NormalizeVector(Vector To Normalize=$Particles.Velocity, Safe Normalize=true, Safe Fallback Threshold=0, Safe Fallback Vector=<0 0 0.001>)) {isSetParameters=true}
   /SetVariables(Particles.ProjectileVelocity:Vector3f=&NormalizeVector(
     Vector To Normalize=$Output.InitializeParticleFromNDC_FortProjectiles.Velocity
     Safe Normalize=true
     Safe Fallback Threshold=0
     Safe Fallback Vector=<0 0 0.001>
    )) {isSetParameters=true}
   /SetVariables(Particles.ProjectileSpeed:NiagaraFloat=&Multiply_Float(
     A=&VectorLength(Vector=$Output.InitializeParticleFromNDC_FortProjectiles.Velocity)
     B=$Engine.DeltaTime
    )) {isSetParameters=true}
   /SetVariables(Particles.RibbonID=$Output.InitializeParticleFromNDC_FortProjectiles.ID) {isSetParameters=true}
   /SetVariables(Particles.RemapByDistance:NiagaraFloat=&RemapRange(
     Input Value To Remap=&FloatFromCurve(
      FloatCurve=DataInterface() {propertyValues {Curve {keys [
          {arriveTangent=0.0002, interpMode=RCIM_Cubic, leaveTangent=0.0002, leaveTangentWeight=0.3333, tangentMode=RCTM_User, time=0, value=0}
          <at 1 0>
          {arriveTangent=0.0001, arriveTangentWeight=833, interpMode=RCIM_Cubic, leaveTangent=0.0001, tangentMode=RCTM_User, time=2500, value=0.35}
          <at 15000 0.8>
          {arriveTangentWeight=6666.667, interpMode=RCIM_Cubic, tangentMode=RCTM_User, time=35000, value=1}
         ]}}}
      CurveIndex=$Particles.LOD Distance
     )
     Input Min=0
     Input Max=1
     Output Min=0.5
     Output Max=3
    )) {isSetParameters=true}
   /SetVariables(Particles.TracerLength:NiagaraFloat=&Multiply_Float(
     A=&VectorLength(Vector=$Particles.Velocity)
     B=&Multiply_Float(
      A=&Lerp_Float(A=1, B=1, Alpha=$Particles.Random)
      B=&SelectFloatFromArray(
       Direct Array Index=$Output.InitializeParticleFromNDC_FortProjectiles.AmmoType
       Float Selection Array=DataInterface() {propertyValues {FloatData [1,0.8,1,0.7]}}
      )
     )
    )) {isSetParameters=true}
   /SetVariables(Particles.TracerLength:NiagaraFloat=&Multiply_Float(
     A=&VectorLength(Vector=$Particles.Velocity)
     B=&Multiply_Float(
      A=&Lerp_Float(A=0.75, B=1, Alpha=$Particles.Random)
      B=&SelectFloatFromArray(
       Direct Array Index=$Output.InitializeParticleFromNDC_FortProjectiles.AmmoType
       Float Selection Array=DataInterface() {propertyValues {FloatData [0.75,0.8,1,0.7]}}
      )
     )
    )) {enabled=false, isSetParameters=true}
   /SetVariables(Particles.TracerLength:NiagaraFloat=&MakeCustomFloatFromBool(
     Bool=$Particles.IsEndPosition
     True Float=$Particles.TracerLength
     False Float=&Multiply_Float(
      A=$Particles.TracerLength
      B=&ClampFloat(Float=&Divide_Float(A=$Particles.ProjectileAge, B=0.01))
     )
    )) {isSetParameters=true}
   /SetVariables(Particles.TracerSoftness:NiagaraFloat=&MakeCustomFloatFromBool(Bool=$Output.InitializeParticleFromNDC_FortProjectiles.bIsSupressed, True Float=0.8, False Float=0.3)) {isSetParameters=true}
   /SetVariables(Particles.Hit Lifetime:NiagaraFloat=&Lerp_Float(
     A=0.1
     B=0.1667
     Alpha=&ClampFloat(Float=&Divide_Float(A=$Particles.Distance, B=$Particles.ProjectileSpeed))
    )) {isSetParameters=true}
   /SetVariables(Particles.Hit Lifetime:NiagaraFloat=&Multiply_Float(A=$Particles.Hit Lifetime, B=10)) {enabled=false, isSetParameters=true}
   &InitializeParticle(
    Color=<rgba 1 0 0.01 1>
    Lifetime=&MakeCustomFloatFromBool(Bool=$Particles.IsEndPosition, True Float=$Particles.Hit Lifetime, False Float=0.0167)
    Mesh Scale Mode=::NewEnumerator3
    Mesh Uniform Scale=100
    Position=$Particles.TargetPosition
    Position Mode=::NewEnumerator1
    Sprite Size=&MakeVector2D(
     X=&SelectFloatFromArray(
      Direct Array Index=$Output.InitializeParticleFromNDC_FortProjectiles.AmmoType
      Float Selection Array=DataInterface() {propertyValues {FloatData [4,18,24,9]}}
     )
     Y=$Particles.TracerLength
    )
    Sprite Size Mode=::NewEnumerator4
    UsePositionOffset=false
    Write Material Random=true
   )
   &InitializeParticle(
    Color=&MakeLinearColorFromVectorAndFloat("Vector (RGB)"=<200 0 0>, "Float (Alpha)"=1)
    Lifetime=&MakeCustomFloatFromBool(Bool=$Particles.IsEndPosition, True Float=$Particles.Hit Lifetime, False Float=0.0167)
    Mesh Scale Mode=::NewEnumerator3
    Mesh Uniform Scale=100
    Position=$Particles.TargetPosition
    Position Mode=::NewEnumerator1
    Sprite Size=&MakeVector2D(
     X=&SelectFloatFromArray(
      Direct Array Index=$Output.InitializeParticleFromNDC_FortProjectiles.AmmoType
      Float Selection Array=DataInterface() {propertyValues {FloatData [8,12,16,6]}}
     )
     Y=$Particles.TracerLength
    )
    Sprite Size Mode=::NewEnumerator4
    UsePositionOffset=false
    Write Material Random=true
   ) {enabled=false, module=InitializeParticle001}
   /SetVariables(Particles.Muzzle Position Offset:Vector3f=&Multiply_VectorByFloat(
     Vector=$Particles.VelocityNormalized
     Float=&Multiply_Float(
      A=&MakeFloatFromVector2d(Vector 2D=$Particles.SpriteSize, Channel=::NewEnumerator1)
      B=&Multiply_Float(A=-0.5, B=1)
     )
    )) {isSetParameters=true}
   &DynamicMaterialParameters(
    Index 0 Param 1=$Particles.TracerSoftness
    Index 0 Param 2=1
    Index 0 Param 3=1
    Module.Index0Param0Write=true
    Module.Index0Param1Write=true
    Module.Index0Param2Write=true
    Module.Index0Param3Write=true
   )
   &SpriteFacingAndAlignment(Sprite Alignment=$Particles.Velocity, WriteAlignment=true, WriteFacing=false) {enabled=false}
  ]
  particleUpdateScript [
   &KillParticles(Kill Particles=$Particles.IsStartPosition)
   &ParticleState()
   &ScaleSpriteSize(
    Scale Factor=&MakeVector2D(
     X=$Particles.RemapByDistance
     Y=&MakeCustomFloatFromBool(
      Bool=$Particles.IsEndPosition
      True Float=&FloatFromCurve(FloatCurve=DataInterface() {propertyValues {Curve {keys [<at 0 0>,<at 0.25 0.66>,<at 0.6 0.66>,<at 1 0>]}}})
      False Float=1
     )
    )
    Scale Sprite Size Mode=::NewEnumerator2
   )
   /SetVariables(Particles.Muzzle Position Offset:Vector3f=&Multiply_VectorByFloat(
     Vector=$Particles.VelocityNormalized
     Float=&Multiply_Float(A=&MakeFloatFromVector2d(Vector 2D=$Particles.SpriteSize, Channel=::NewEnumerator1), B=-0.5)
    )) {isSetParameters=true}
   /SetVariables(Particles.Position=&AddVectorToPosition(Position=$Particles.Initial.Position, Vector=$Particles.Muzzle Position Offset)) {isSetParameters=true}
   /SetVariables(Particles.Position=&MakeCustomVectorFromBool(
     Bool=$Particles.IsEndPosition
     True Vector=&Lerp_Vector(
      A=Linked() {linkedVariable {name=Particles.PreviousPosition, type=NiagaraPosition}}
      B=Linked() {linkedVariable {name=Particles.Position, type=NiagaraPosition}}
      Alpha=&FloatFromCurve(FloatCurve=DataInterface() {propertyValues {Curve {keys [
           {arriveTangent=1.6275, interpMode=RCIM_Cubic, leaveTangent=1.6275, tangentMode=RCTM_User, time=0, value=0}
           {interpMode=RCIM_Cubic, tangentMode=RCTM_User, time=0.66, value=1}
          ]}}})
     )
     False Vector=Linked() {linkedVariable {name=Particles.Position, type=NiagaraPosition}}
    )) {isSetParameters=true}
   &ScaleColor(Scale RGB=&VectorFromFloat(Value=30), ScaleA=false, ScaleRGB=true)
  ]
  rendererClasses [&NiagaraMeshRendererProperties, &NiagaraSpriteRendererProperties]
  renderers [
   {rendererClass=&NiagaraMeshRendererProperties, rendererIndex=0}
   {rendererClass=&NiagaraSpriteRendererProperties, rendererIndex=1}
   {rendererClass=&NiagaraSpriteRendererProperties, rendererIndex=2}
  ]
 }}
summary {
 systemName=NS_NDC_ProjectilesHandler_Energy
 emitters [{emitterName=Bullet, rendererClasses [&NiagaraMeshRendererProperties, &NiagaraSpriteRendererProperties]}]
 userVariables []
}
