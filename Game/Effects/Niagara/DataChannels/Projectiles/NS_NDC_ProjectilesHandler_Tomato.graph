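# Niagara system that draws Tomato-variant projectile tracers for many shots at once, reading projectile data from a Niagara Data Channel.
# Only the Bullet emitter is enabled; Trail, Empty and debug emitters are disabled.
# Bullet spawns particles from channel entries, stretching mesh and sprite tracers by per-frame travel, scaling width by LOD distance and tinting by projectile variant.
# Start-position particles are killed, while end-position particles linger briefly, sliding to the hit point and shrinking.
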NiagaraSystem /Game/Effects/Niagara/DataChannels/Projectiles/NS_NDC_ProjectilesHandler_Tomato
dependencies {
 usedDataInterfaces [
  &NiagaraDataInterfaceArrayFloat
  &NiagaraDataInterfaceCurve
  /Script/Niagara.NiagaraDataInterfaceDataChannelRead
 ]
 usedDynamicInputs []
 usedModules [
  &SystemState
  /Niagara/Modules/System/CompleteIfUnused
  &self:SystemSpawnScript.SystemScriptSource.SystemScriptGraph.NiagaraNodeAssignment_0.SetVariables_#1
 ]
}
emitters {
 Bullet {
  emitterSpawnScript [/SetVariables(Emitter.UseTeardrop=true) {isSetParameters=true}]
  emitterUpdateScript [
   &EmitterState()
   &/Spawn_NDCFortProjectiles(Max Spawn Count=1, Min Spawn Count=1, Reader:NiagaraDataInterfaceDataChannelRead=$System.NDCReader)
   /SetVariables(Emitter.InnerColor:LinearColor=&LerpLinearColors(StartColor=<rgba 0.5 0.0083 0 1>, EndColor=<rgba 0 0.3253 0.5 1>, LerpFactor=&MakeFloatFromInt(Int=0))) {enabled=false, isSetParameters=true}
  ]
  particleSpawnScript [
   &/InitializeParticleFromNDC_FortProjectiles(FallbackAmmoType=2)
   &self:ScratchModule() {module=ScratchModule}
   /SetVariables(Particles.PositionType:NiagaraInt32=$Output.InitializeParticleFromNDC_FortProjectiles.PositionType) {isSetParameters=true}
   /SetVariables(Particles.Random=&RandomRangeFloat(
     Minimum=0
     Maximum=1
     Randomness Mode=::NewEnumerator1
     Random Seed=&ReturnParticleIDAttributes(Particle ID Attribute=::NewEnumerator1, Niagara ID=$Particles.RibbonID)
     Fixed Random Seed=true
    )) {isSetParameters=true}
   /SetVariables(Particles.LOD Distance:NiagaraFloat=$Output.InitializeParticleFromNDC_FortProjectiles.LODDistance) {isSetParameters=true}
   /SetVariables(Particles.AmmoTypeScale:NiagaraFloat=&SelectFloatFromArray(
     Direct Array Index=$Output.InitializeParticleFromNDC_FortProjectiles.AmmoType
     Float Selection Array=DataInterface() {propertyValues {FloatData [0.75,1,1.2,0.75]}}
    )) {isSetParameters=true}
   /SetVariables(Particles.IsStartPosition:NiagaraBool=&SetBoolByIntComparison(A=$Particles.PositionType, B=0, Comparison Type=::NewEnumerator4)) {isSetParameters=true}
   /SetVariables(Particles.IsEndPosition:NiagaraBool=&SetBoolByIntComparison(A=$Particles.PositionType, B=2, Comparison Type=::NewEnumerator4)) {isSetParameters=true}
   /SetVariables(Particles.OriginTime:NiagaraFloat=$Output.InitializeParticleFromNDC_FortProjectiles.OriginTime) {isSetParameters=true}
   /SetVariables(Particles.ProjectileAge:NiagaraFloat=&Subtract_Float(A=$Engine.Time, B=$Particles.OriginTime)) {isSetParameters=true}
   /SetVariables(Particles.MuzzlePosition:NiagaraPosition=$Output.InitializeParticleFromNDC_FortProjectiles.OriginPosition) {isSetParameters=true}
   /SetVariables(Particles.PreviousPosition:NiagaraPosition=$Output.InitializeParticleFromNDC_FortProjectiles.PrevPosition) {isSetParameters=true}
   /SetVariables(Particles.TargetPosition:NiagaraPosition=$Output.InitializeParticleFromNDC_FortProjectiles.Position) {isSetParameters=true}
   /SetVariables(Particles.RibbonUVDistance:NiagaraFloat=&DistanceBetweenPositions(A=$Particles.TargetPosition, B=$Particles.MuzzlePosition)) {isSetParameters=true}
   /SetVariables(Particles.Distance:NiagaraFloat=&DistanceBetweenPositions(A=$Particles.PreviousPosition, B=$Particles.TargetPosition)) {isSetParameters=true}
   /SetVariables(Particles.Velocity=&Subtract_Position(A=$Particles.TargetPosition, B=$Particles.PreviousPosition)) {isSetParameters=true}
   /SetVariables(Particles.VelocityNormalized:Vector3f=&NormalizeVector(Vector To Normalize=$Particles.Velocity, Safe Normalize=true, Safe Fallback Threshold=0, Safe Fallback Vector=<0 0 0.001>)) {isSetParameters=true}
   /SetVariables(Particles.ProjectileVelocity:Vector3f=&NormalizeVector(
     Vector To Normalize=$Output.InitializeParticleFromNDC_FortProjectiles.Velocity
     Safe Normalize=true
     Safe Fallback Threshold=0
     Safe Fallback Vector=<0 0 0.001>
    )) {isSetParameters=true}
   /SetVariables(Particles.ProjectileSpeed:NiagaraFloat=&Multiply_Float(
     A=&VectorLength(Vector=$Output.InitializeParticleFromNDC_FortProjectiles.Velocity)
     B=$Engine.DeltaTime
    )) {isSetParameters=true}
   /SetVariables(Particles.RibbonID=$Output.InitializeParticleFromNDC_FortProjectiles.ID) {isSetParameters=true}
   /SetVariables(Particles.RemapByDistance:NiagaraFloat=&RemapRange(
     Input Value To Remap=&FloatFromCurve(
      FloatCurve=DataInterface() {propertyValues {Curve {keys [
          {arriveTangent=0.0002, interpMode=RCIM_Cubic, leaveTangent=0.0002, leaveTangentWeight=0.3333, tangentMode=RCTM_User, time=0, value=0}
          <at 1 0>
          {arriveTangent=0.0001, arriveTangentWeight=833, interpMode=RCIM_Cubic, leaveTangent=0.0001, tangentMode=RCTM_User, time=2500, value=0.35}
          <at 15000 0.8>
          {arriveTangentWeight=6666.667, interpMode=RCIM_Cubic, tangentMode=RCTM_User, time=35000, value=1}
         ]}}}
      CurveIndex=$Particles.LOD Distance
     )
     Input Min=0
     Input Max=1
     Output Min=0.5
     Output Max=3
    )) {isSetParameters=true}
   /SetVariables(Particles.TracerLength:NiagaraFloat=&Multiply_Float(
     A=&VectorLength(Vector=$Particles.Velocity)
     B=&Multiply_Float(
      A=&Lerp_Float(A=1, B=1, Alpha=$Particles.Random)
      B=&SelectFloatFromArray(
       Direct Array Index=$Output.InitializeParticleFromNDC_FortProjectiles.AmmoType
       Float Selection Array=DataInterface() {propertyValues {FloatData [1,0.8,1,0.7]}}
      )
     )
    )) {isSetParameters=true}
   /SetVariables(Particles.TracerLength:NiagaraFloat=&Multiply_Float(
     A=&VectorLength(Vector=$Particles.Velocity)
     B=&Multiply_Float(
      A=&Lerp_Float(A=0.75, B=1, Alpha=$Particles.Random)
      B=&SelectFloatFromArray(
       Direct Array Index=$Output.InitializeParticleFromNDC_FortProjectiles.AmmoType
       Float Selection Array=DataInterface() {propertyValues {FloatData [0.75,0.8,1,0.7]}}
      )
     )
    )) {enabled=false, isSetParameters=true}
   /SetVariables(Particles.TracerLength:NiagaraFloat=&MakeCustomFloatFromBool(
     Bool=$Particles.IsEndPosition
     True Float=$Particles.TracerLength
     False Float=&Multiply_Float(
      A=$Particles.TracerLength
      B=&ClampFloat(Float=&Divide_Float(A=$Particles.ProjectileAge, B=0.01))
     )
    )) {isSetParameters=true}
   /SetVariables(Particles.TracerSoftness:NiagaraFloat=&MakeCustomFloatFromBool(Bool=$Output.InitializeParticleFromNDC_FortProjectiles.bIsSupressed, True Float=0.8, False Float=1)) {isSetParameters=true}
   /SetVariables(Particles.Hit Lifetime:NiagaraFloat=&Lerp_Float(
     A=0.1
     B=0.1667
     Alpha=&ClampFloat(Float=&Divide_Float(A=$Particles.Distance, B=$Particles.ProjectileSpeed))
    )) {isSetParameters=true}
   /SetVariables(Particles.Hit Lifetime:NiagaraFloat=&Multiply_Float(A=$Particles.Hit Lifetime, B=10)) {enabled=false, isSetParameters=true}
   &InitializeParticle(
    Color=&LerpLinearColors(
     StartColor=<rgba 1 0.1 0 1>
     EndColor=<rgba 0 0.0973 0.73 1>
     LerpFactor=&MakeFloatFromInt(Int=$Particles.ProjectileVariant)
    )
    Lifetime=&MakeCustomFloatFromBool(Bool=$Particles.IsEndPosition, True Float=$Particles.Hit Lifetime, False Float=0.0167)
    Mesh Scale Mode=::NewEnumerator3
    Mesh Uniform Scale=100
    Position=$Particles.TargetPosition
    Position Mode=::NewEnumerator1
    Sprite Size=&MakeVector2D(X=9, Y=$Particles.TracerLength)
    Sprite Size Mode=::NewEnumerator4
    UsePositionOffset=false
    Write Material Random=true
   )
   &InitializeParticle(
    Color=&MakeLinearColorFromVectorAndFloat("Vector (RGB)"=<200 0 0>, "Float (Alpha)"=1)
    Lifetime=&MakeCustomFloatFromBool(Bool=$Particles.IsEndPosition, True Float=$Particles.Hit Lifetime, False Float=0.0167)
    Mesh Scale Mode=::NewEnumerator3
    Mesh Uniform Scale=100
    Position=$Particles.TargetPosition
    Position Mode=::NewEnumerator1
    Sprite Size=&MakeVector2D(
     X=&SelectFloatFromArray(
      Direct Array Index=$Output.InitializeParticleFromNDC_FortProjectiles.AmmoType
      Float Selection Array=DataInterface() {propertyValues {FloatData [8,12,16,6]}}
     )
     Y=$Particles.TracerLength
    )
    Sprite Size Mode=::NewEnumerator4
    UsePositionOffset=false
    Write Material Random=true
   ) {enabled=false, module=InitializeParticle001}
   /SetVariables(Particles.Muzzle Position Offset:Vector3f=&Multiply_VectorByFloat(
     Vector=$Particles.VelocityNormalized
     Float=&Multiply_Float(
      A=&MakeFloatFromVector2d(Vector 2D=$Particles.SpriteSize, Channel=::NewEnumerator1)
      B=&Multiply_Float(A=-0.5, B=1)
     )
    )) {isSetParameters=true}
   &DynamicMaterialParameters(
    Index 0 Param 1=$Particles.TracerSoftness
    Index 0 Param 2=1
    Index 0 Param 3=1
    Module.Index0Param0Write=true
    Module.Index0Param1Write=true
    Module.Index0Param2Write=true
    Module.Index0Param3Write=true
   )
   &SpriteFacingAndAlignment(Sprite Alignment=$Particles.Velocity, WriteAlignment=true, WriteFacing=false) {enabled=false}
   /SetVariables(Particles.VisibilityTag=$Particles.ProjectileVariant) {isSetParameters=true}
  ]
  particleUpdateScript [
   &KillParticles(Kill Particles=$Particles.IsStartPosition)
   &ParticleState()
   &ScaleSpriteSize(
    Scale Factor=&MakeVector2D(
     X=$Particles.RemapByDistance
     Y=&MakeCustomFloatFromBool(
      Bool=$Particles.IsEndPosition
      True Float=&FloatFromCurve(FloatCurve=DataInterface() {propertyValues {Curve {keys [<at 0 0>,<at 0.25 0.66>,<at 0.6 0.66>,<at 1 0>]}}})
      False Float=1
     )
    )
    Scale Sprite Size Mode=::NewEnumerator2
   )
   /SetVariables(Particles.Muzzle Position Offset:Vector3f=&Multiply_VectorByFloat(
     Vector=$Particles.VelocityNormalized
     Float=&Multiply_Float(A=&MakeFloatFromVector2d(Vector 2D=$Particles.SpriteSize, Channel=::NewEnumerator1), B=-0.5)
    )) {isSetParameters=true}
   /SetVariables(Particles.Position=&AddVectorToPosition(Position=$Particles.Initial.Position, Vector=$Particles.Muzzle Position Offset)) {isSetParameters=true}
   /SetVariables(Particles.Position=&MakeCustomVectorFromBool(
     Bool=$Particles.IsEndPosition
     True Vector=&Lerp_Vector(
      A=Linked() {linkedVariable {name=Particles.PreviousPosition, type=NiagaraPosition}}
      B=Linked() {linkedVariable {name=Particles.Position, type=NiagaraPosition}}
      Alpha=&FloatFromCurve(FloatCurve=DataInterface() {propertyValues {Curve {keys [
           {arriveTangent=1.6275, interpMode=RCIM_Cubic, leaveTangent=1.6275, tangentMode=RCTM_User, time=0, value=0}
           {interpMode=RCIM_Cubic, tangentMode=RCTM_User, time=0.66, value=1}
          ]}}})
     )
     False Vector=Linked() {linkedVariable {name=Particles.Position, type=NiagaraPosition}}
    )) {isSetParameters=true}
   &ScaleColor(
    Scale RGB=&VectorFromFloat(Value=&Multiply_Float(
      A=&Lerp_Float(A=30, B=10, Alpha=&MakeFloatFromInt(Int=$Particles.ProjectileVariant))
      B=&FloatFromCurve(FloatCurve=DataInterface() {propertyValues {Curve {keys [<at 0 3>,<at 0.3 1>]}}})
     ))
    ScaleA=false
    ScaleRGB=true
   )
  ]
  rendererClasses [&NiagaraMeshRendererProperties, &NiagaraSpriteRendererProperties]
  renderers [
   {rendererClass=&NiagaraMeshRendererProperties, rendererIndex=0}
   {rendererClass=&NiagaraSpriteRendererProperties, rendererIndex=1}
   {rendererClass=&NiagaraSpriteRendererProperties, rendererIndex=2}
   {rendererClass=&NiagaraSpriteRendererProperties, rendererIndex=3}
   {rendererClass=&NiagaraSpriteRendererProperties, rendererIndex=4}
  ]
 }
 Empty {
  bEnabled=false
  emitterUpdateScript [
   &EmitterState()
   &Spawn_NDCFortProjectiles(Reader:NiagaraDataInterfaceDataChannelRead=$System.NDCReader)
  ]
  particleSpawnScript [
   &InitializeParticleFromNDC_FortProjectiles()
   &InitializeParticle(
    Color=<rgba 12 5.5739 0.3723 1>
    Position=$Output.InitializeParticleFromNDC_FortProjectiles.Position
    Position Mode=::NewEnumerator1
    Position Offset=&Multiply_VectorByFloat(
     Vector=$Output.InitializeParticleFromNDC_FortProjectiles.OriginOffset
     Float=&FloatFromCurve(
      FloatCurve=DataInterface() {propertyValues {Curve {keys [{arriveTangent=-0.4167, interpMode=RCIM_Cubic, time=0, value=1}, {interpMode=RCIM_Cubic, leaveTangent=-26.8946, time=0.05, value=0}]}}}
      CurveIndex=$Particles.ProjectileAge
     )
    )
    Ribbon Width=&Multiply_Float(A=$Particles.TracerWidth, B=&Lerp_Float(A=0.75, B=1, Alpha=$Particles.Random))
    Ribbon Width Mode=::NewEnumerator1
    UsePositionOffset=true
    Write Material Random=true
   )
  ]
  particleUpdateScript [
   &ParticleState()
   &KillParticles(Kill Particles=&Boolean_NotOperation(Bool=$Output.ParticleState.FirstFrame))
  ]
  rendererClasses [&NiagaraSpriteRendererProperties]
  renderers [{rendererClass=&NiagaraSpriteRendererProperties}]
 }
 FN_Debug {
  bEnabled=false
  emitterUpdateScript [
   &EmitterState(Life Cycle Mode=::NewEnumerator1, Loop Duration=1, UseLoopDelay=false)
   &SpawnBurst_Instantaneous(Spawn Count=1, Use Spawn Probability=false)
   &Spawn_NDCFortProjectiles(Max Spawn Count=1, Min Spawn Count=1, Reader:NiagaraDataInterfaceDataChannelRead=$System.NDCReader)
  ]
  particleSpawnScript [
   /Niagara/Modules/Spawn/Initialization/InitializeParticle(
    Lifetime=5
    Module.Write Color=true
    Module.Write Lifetime=true
    Module.Write Mass=true
    Module.Write Material Random=true
    Module.Write Position=false
    Module.Write Scale=false
    Module.Write SpriteRotation=false
    Module.Write SpriteSize=true
    Position=&SimulationPosition()
    Sprite Size=<100 100>
   )
   &InitializeParticleFromNDC_FortProjectiles(FallbackAmmoType=2)
  ]
  particleUpdateScript [
   &ParticleState()
   /SetVariables(Particles.Debug:NiagaraFloat=&MakeFloatFromInt(Int=$Output.InitializeParticleFromNDC_FortProjectiles.AmmoType)) {isSetParameters=true}
   &DynamicMaterialParameters(
    Index 0 Param 1=$Particles.Debug
    Module.Param0WriteEnabled=true
    Module.Param1WriteEnabled=true
    Module.Param2WriteEnabled=true
    Module.Param3WriteEnabled=true
   )
  ]
  rendererClasses [&NiagaraSpriteRendererProperties]
  renderers [{rendererClass=&NiagaraSpriteRendererProperties}]
 }
 Trail {
  bEnabled=false
  emitterUpdateScript [
   &EmitterState()
   &Spawn_NDCFortProjectiles(Max Spawn Count=1, Min Spawn Count=1, Reader:NiagaraDataInterfaceDataChannelRead=$System.NDCReader)
  ]
  particleSpawnScript [
   &InitializeParticleFromNDC_FortProjectiles(FallbackAmmoType=2)
   /SetVariables(Particles.PositionType:NiagaraInt32=$Output.InitializeParticleFromNDC_FortProjectiles.PositionType) {isSetParameters=true}
   /SetVariables(Particles.LOD Distance:NiagaraFloat=$Output.InitializeParticleFromNDC_FortProjectiles.LODDistance) {isSetParameters=true}
   /SetVariables(Particles.AmmoTypeScale:NiagaraFloat=&SelectFloatFromArray(
     Direct Array Index=$Output.InitializeParticleFromNDC_FortProjectiles.AmmoType
     Float Selection Array=DataInterface() {propertyValues {FloatData [0.75,1,1.2,0.75]}}
    )) {isSetParameters=true}
   /SetVariables(Particles.IsEndPosition:NiagaraBool=&SetBoolByIntComparison(A=$Particles.PositionType, B=2, Comparison Type=::NewEnumerator4)) {isSetParameters=true}
   /SetVariables(Particles.IsStartPosition:NiagaraBool=&SetBoolByIntComparison(A=$Particles.PositionType, B=0, Comparison Type=::NewEnumerator4)) {isSetParameters=true}
   /SetVariables(Particles.Velocity=&Subtract_Position(A=$Particles.TargetPosition, B=$Particles.PreviousPosition)) {enabled=false, isSetParameters=true}
   /SetVariables(Particles.ProjectileVelocity:Vector3f=$Output.InitializeParticleFromNDC_FortProjectiles.Velocity) {isSetParameters=true}
   /SetVariables(Particles.ProjectileSpeed:NiagaraFloat=&Multiply_Float(
     A=&VectorLength(Vector=$Output.InitializeParticleFromNDC_FortProjectiles.Velocity)
     B=$Engine.DeltaTime
    )) {isSetParameters=true}
   /SetVariables(Particles.OriginTime:NiagaraFloat=$Output.InitializeParticleFromNDC_FortProjectiles.OriginTime) {isSetParameters=true}
   /SetVariables(Particles.ProjectileAge:NiagaraFloat=&Subtract_Float(A=$Engine.Time, B=$Particles.OriginTime)) {isSetParameters=true}
   /SetVariables(Particles.MuzzlePosition:NiagaraPosition=$Output.InitializeParticleFromNDC_FortProjectiles.OriginPosition) {isSetParameters=true}
   /SetVariables(Particles.TargetPosition:NiagaraPosition=$Output.InitializeParticleFromNDC_FortProjectiles.Position) {isSetParameters=true}
   /SetVariables(Particles.RibbonID=$Output.InitializeParticleFromNDC_FortProjectiles.ID) {isSetParameters=true}
   /SetVariables(Particles.Random=&RandomRangeFloat(
     Minimum=0
     Maximum=1
     Randomness Mode=::NewEnumerator1
     Random Seed=&ReturnParticleIDAttributes(Particle ID Attribute=::NewEnumerator1, Niagara ID=$Particles.RibbonID)
     Fixed Random Seed=true
    )) {isSetParameters=true}
   /SetVariables(Particles.RandomLerpAge:NiagaraFloat=&MakeCustomFloatFromBool(Bool=$Particles.IsEndPosition, True Float=0.0167, False Float=0.0333)) {isSetParameters=true}
   /SetVariables(Particles.LerpAge:NiagaraFloat=&RemapRange(Input Value To Remap=$Particles.ProjectileAge, Input Min=0, Input Max=$Particles.RandomLerpAge, Output Min=0, Output Max=1)) {isSetParameters=true}
   /SetVariables(Particles.Distance:NiagaraFloat=&DistanceBetweenPositions(A=$Output.InitializeParticleFromNDC_FortProjectiles.PrevPosition, B=$Output.InitializeParticleFromNDC_FortProjectiles.Position)) {isSetParameters=true}
   /SetVariables(Particles.RibbonUVDistance:NiagaraFloat=&DistanceBetweenPositions(A=$Particles.TargetPosition, B=$Particles.MuzzlePosition)) {isSetParameters=true}
   /SetVariables(Particles.RemapByDistance:NiagaraFloat=&RemapRange(
     Input Value To Remap=&FloatFromCurve(
      FloatCurve=DataInterface() {propertyValues {Curve {keys [
          {arriveTangent=0.0002, interpMode=RCIM_Cubic, leaveTangent=0.0002, leaveTangentWeight=0.3333, tangentMode=RCTM_User, time=0, value=0}
          <at 1 0>
          {arriveTangent=0.0001, arriveTangentWeight=833, interpMode=RCIM_Cubic, leaveTangent=0.0001, tangentMode=RCTM_User, time=2500, value=0.35}
          <at 15000 0.8>
          {arriveTangentWeight=6666.667, interpMode=RCIM_Cubic, tangentMode=RCTM_User, time=35000, value=1}
         ]}}}
      CurveIndex=$Particles.LOD Distance
     )
     Input Min=0
     Input Max=1
     Output Min=1
     Output Max=2
    )) {isSetParameters=true}
   /SetVariables(
    Particles.TracerMultiplier:NiagaraFloat=&MakeCustomFloatFromBool(Bool=$Output.InitializeParticleFromNDC_FortProjectiles.bIsSupressed, True Float=0.15, False Float=0.5)
    Particles.TracerWidth:NiagaraFloat=&Multiply_Float(A=&Multiply_Float(A=10, B=$Particles.AmmoTypeScale), B=$Particles.RemapByDistance)
   ) {isSetParameters=true}
   /Niagara/Modules/Spawn/Initialization/V2/InitializeParticle(
    Color=&MakeLinearColorFromVectorAndFloat(
     "Vector (RGB)"=<1 0.15 0>
     "Float (Alpha)"=&MakeCustomFloatFromBool(Bool=$Output.InitializeParticleFromNDC_FortProjectiles.bIsSupressed, True Float=0.4, False Float=1)
    )
    Lifetime=&Multiply_Float(
     A=1
     B=&SelectFloatFromArray(
      Direct Array Index=$Output.InitializeParticleFromNDC_FortProjectiles.AmmoType
      Float Selection Array=DataInterface() {propertyValues {FloatData [0.4,1,2,0.7]}}
     )
    )
    Position=$Output.InitializeParticleFromNDC_FortProjectiles.Position
    Position Mode=::NewEnumerator1
    Ribbon Width=&Multiply_Float(A=$Particles.TracerWidth, B=&Lerp_Float(A=0.75, B=1, Alpha=$Particles.Random))
    Ribbon Width Mode=::NewEnumerator1
    UsePositionOffset=false
    Write Material Random=true
   )
   &ApplyInitialForces()
  ]
  particleUpdateScript [
   &ParticleState()
   &ScaleColor(
    Scale Alpha=&FloatFromCurve(FloatCurve=DataInterface() {propertyValues {Curve {keys [{arriveTangent=-2.5785, interpMode=RCIM_Cubic, time=0, value=1}, {interpMode=RCIM_Cubic, tangentMode=RCTM_User, time=1, value=0}]}}})
    Scale RGB=&Multiply_VectorByFloat(
     Vector=<1 1 1>
     Float=&FloatFromCurve(
      FloatCurve=DataInterface() {propertyValues {Curve {keys [{arriveTangent=-9.6343, interpMode=RCIM_Cubic, time=0, value=1}, <at 0.2 0 RCIM_Cubic>]}}}
      Scale Curve=$Particles.TracerMultiplier
     )
    )
    ScaleA=true
    ScaleRGB=true
   )
   &ScaleRibbonWidth(Ribbon Width Scale=&FloatFromCurve(FloatCurve=DataInterface() {propertyValues {Curve {keys [<at 0 0.4 RCIM_Cubic>, {tangentMode=RCTM_User, time=1, value=1}]}}}))
   &DynamicMaterialParameters(
    Index 0 Param 1=&FloatFromCurve(
     FloatCurve=DataInterface() {propertyValues {Curve {keys [<at 0.25 0 RCIM_Cubic>, {interpMode=RCIM_Cubic, tangentMode=RCTM_User, time=1, value=1}]}}}
     Scale Curve=&Multiply_Float(A=4, B=$Particles.AmmoTypeScale)
    )
    Index 0 Param 2=&FloatFromCurve(
     FloatCurve=DataInterface() {propertyValues {Curve {keys [
         {arriveTangent=4.1785, interpMode=RCIM_Cubic, leaveTangent=4.1785, tangentMode=RCTM_User, time=0, value=0}
         {arriveTangent=1.1179, interpMode=RCIM_Cubic, leaveTangent=1.1179, tangentMode=RCTM_User, time=0.2042, value=0.6228}
         {arriveTangent=0.0224, interpMode=RCIM_Cubic, leaveTangent=0.0224, tangentMode=RCTM_User, time=1, value=1}
        ]}}}
     Scale Curve=-2.5
    )
    Index 0 Param 3=$Particles.Random
    Index 0 Param 4=&FloatFromCurve(
     FloatCurve=DataInterface() {propertyValues {Curve {keys [{interpMode=RCIM_Cubic, tangentMode=RCTM_User, time=0, value=0}, {interpMode=RCIM_Cubic, tangentMode=RCTM_User, time=0.33, value=1}]}}}
     Scale Curve=0.8
    )
    Module.Index0Param0Write=true
    Module.Index0Param1Write=true
    Module.Index0Param2Write=true
    Module.Index0Param3Write=true
   )
  ]
  rendererClasses [
   &NiagaraRibbonRendererProperties
   &NiagaraSpriteRendererProperties
   &NiagaraMeshRendererProperties
  ]
  renderers [
   {rendererClass=&NiagaraRibbonRendererProperties}
   {rendererClass=&NiagaraSpriteRendererProperties, rendererIndex=1}
   {rendererClass=&NiagaraMeshRendererProperties, rendererIndex=2}
  ]
 }
}
summary {
 systemName=NS_NDC_ProjectilesHandler_Tomato
 emitters [
  {
   bEnabled=false
   emitterName=Trail
   rendererClasses [
    &NiagaraRibbonRendererProperties
    &NiagaraSpriteRendererProperties
    &NiagaraMeshRendererProperties
   ]
  }
  {emitterName=Bullet, rendererClasses [&NiagaraMeshRendererProperties, &NiagaraSpriteRendererProperties]}
  {bEnabled=false, emitterName=Empty, rendererClasses [&NiagaraSpriteRendererProperties]}
  {bEnabled=false, emitterName=FN_Debug, rendererClasses [&NiagaraSpriteRendererProperties]}
 ]
 userVariables []
}
